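# Masked, lit Niagara sprite material for a toon droplet in the IronLog weapon's muzzle flash effect.
# Particle alpha offsets cell noise to form a dissolve factor that blends in the water-drop shape and a softened drop normal over a flat normal.
# Particle color, scaled down, feeds the shared MF_FX_EmissiveAdjustments function.
# An inverted Frac(Time) chain feeds no other node.
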Material /WildEstateWeapons/Gameplay/IronLog/FX/Muzzle/FNE_MuzzleFlash/M_IronLog_ToonDropletFNE
graph {
 properties {bAutomaticallySetUsageInEditor=false, bUsedWithNiagaraSprites=true, blendMode=BLEND_Masked, usedShadingModels=Default Lit}
 nodes [
  n1 = Constant3Vector() {constant=<rgba 0 0 1 0>}
  n2 = ParticleColor()
  n3 = Subtract(A=TextureSample() {texture=/Game/Global/Textures/Default/Noise/T_CellNoise_01}.R, B=LinearInterpolate(Alpha=@n2.A) {constA=1, constB=-1})
  LinearInterpolate(
   A=@n1
   B=LinearInterpolate(
    A=@n1
    B=TextureSample() {samplerType=SAMPLERTYPE_Normal, texture=/Game/Packages/Fortress_Transylvania/Haunted_Forest/Swamps/Textures/T_Water_drop_02_n}.RGB
   )
   Alpha=@n3
  )
  LinearInterpolate(B=TextureSample() {texture=/Game/Packages/Fortress_Transylvania/Haunted_Forest/Swamps/Textures/T_Water_drop_02}.R, Alpha=@n3)
  MaterialFunctionCall(Emissive=Divide(A=@n2.RGB) {constB=10}) {materialFunction=&MF_FX_EmissiveAdjustments}
  OneMinus(None=Frac(None=Time()))
 ]
}
